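# Niagara system for the Doodling emote gadget prop, producing a light, sporadic splash of small pinkish droplets.
# A single sprite emitter, Droplets_Splash, spawns a few short-lived elongated particles per second at a randomized rate, launched in a mostly upward cone.
# Gravity and drag arc the droplets while a curve shrinks them to zero over their lifetime.
# User skeletal mesh and ribbon ID parameters are declared but not referenced by the emitter's modules.
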NiagaraSystem /BRCosmetics/Animation/Props/Gadgets_Doodling/FX/NS_Emote_Doodling
dependencies {usedDataInterfaces [&NiagaraDataInterfaceCurve], usedDynamicInputs [], usedModules [&SystemState]}
emitters {Droplets_Splash {
  emitterUpdateScript [
   &EmitterState(Life Cycle Mode=::NewEnumerator1, Loop Behavior=::NewEnumerator1, Loop Duration Mode=::NewEnumerator1, UseLoopDelay=false)
   &SpawnRate(SpawnRate=&RandomRangeFloat(Minimum=2, Maximum=6), Use Spawn Probability=false)
  ]
  particleSpawnScript [
   &InitializeParticle(
    Lifetime=&RandomRangeFloat(Minimum=0.333, Maximum=1)
    Sprite Size=<1.5 3>
    Sprite Size Mode=::NewEnumerator4
    UsePositionOffset=false
    Write Material Random=true
   )
   &AddVelocityInCone(
    Cone Axis=<1 -1 6>
    Cone Axis Coordinate Space=::Simulation
    Module.Use Velocity Falloff On Cone Axis=false
    Velocity Strength=&RandomRangeFloat(Minimum=55, Maximum=135)
   )
  ]
  particleUpdateScript [
   &GravityForce(Gravity=<0 0 -222>)
   &Drag(Use Linear Drag=true, Use Rotational Drag=false)
   &SolveForcesAndVelocity(Clamp Velocity=false, Limit Acceleration=false)
   &ParticleState()
   &ScaleSpriteSize(Scale Factor=&Vector2DFromFloat(Value=&FloatFromCurve(FloatCurve=DataInterface() {propertyValues {Curve {keys [<at 0 1>,<at 1 0>]}}}, Scale Curve=1.3)))
   &Color(Color=<rgba 0.6654 0.2024 0.2145 0>)
  ]
  rendererClasses [&NiagaraSpriteRendererProperties]
  renderers [{rendererClass=&NiagaraSpriteRendererProperties}]
 }}
summary {
 systemName=NS_Emote_Doodling
 emitters [{emitterName=Droplets_Splash, rendererClasses [&NiagaraSpriteRendererProperties]}]
 userVariables [
  User.RibbonIDOffset {type=NiagaraInt32}
  User.SkeletalMesh {
   type=NiagaraDataInterfaceSkeletalMesh
   defaultValue {
    struct=NiagaraExt_VariableValue_DataInterface
    value {dataInterface=&self:NiagaraDataInterfaceSkeletalMesh_0}
   }
  }
 ]
}
