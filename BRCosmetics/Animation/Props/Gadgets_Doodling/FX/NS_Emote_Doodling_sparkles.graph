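# Niagara sparkle effect for the Doodling gadget emote, with a single Sparks emitter continuously spawning short-lived pink or blue sprites in an offset sphere.
# One random value drives lifetime, and secondary smaller, near-transparent "Sphere" color and size attributes accompany two sprite renderers.
# Drag and curl noise move particles while curves fade and scale them.
# The burst spawner is disabled, and the exposed SkeletalMesh and RibbonIDOffset parameters appear unused.
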NiagaraSystem /BRCosmetics/Animation/Props/Gadgets_Doodling/FX/NS_Emote_Doodling_sparkles
dependencies {usedDataInterfaces [&NiagaraDataInterfaceCurve], usedDynamicInputs [], usedModules [&SystemState]}
emitters {Sparks {
  emitterUpdateScript [
   &EmitterState()
   &SpawnRate(
    SpawnRate=&RandomRangeFloat(Minimum=6, Maximum=15, Evaluation Type=::NewEnumerator1)
    Use Spawn Probability=false
   )
   &SpawnBurst_Instantaneous(
    Spawn Count=&RandomRangeInt(Minimum=2, Maximum=5)
    Spawn Time=&RandomRangeFloat(Minimum=0.05, Maximum=0.15)
    Use Spawn Probability=false
   ) {enabled=false}
  ]
  particleSpawnScript [
   /SetVariables(Particles.MaterialRandom:NiagaraFloat=&RandomRangeFloat(Minimum=0, Maximum=1)) {isSetParameters=true}
   /Niagara/Modules/Spawn/Initialization/InitializeParticle(
    Color=&MakeCustomLinearColorFromBool(Bool=&RandomBool(), True Color=<rgba 1 0.1411 0.5047 1>, False Color=<rgba 0.15 0.314 0.6172 1>)
    Lifetime=&Lerp_Float(A=0.3, B=0.9, Alpha=$Particles.MaterialRandom)
    Module.Write Color=true
    Module.Write Lifetime=true
    Module.Write Mass=true
    Module.Write Material Random=false
    Module.Write Position=true
    Module.Write Scale=false
    Module.Write SpriteRotation=true
    Module.Write SpriteSize=true
    Sprite Rotation=&RandomRangeFloat(Minimum=0, Maximum=359)
    Sprite Size=&Vector2DFromFloat(Value=&RandomRangeFloat(Minimum=8, Maximum=10))
   )
   /SetVariables(
    Particles.Color_Sphere:LinearColor=&MultiplyLinearColor(Linear Color=$Particles.Color, Scale Factor=<1 1 1 0.01>)
    Particles.SpriteSize_Sphere:Vector2f=&Multiply_Vector2(A=$Particles.SpriteSize, B=&Vector2DFromFloat(Value=0.7))
   ) {isSetParameters=true}
   &SphereLocation(
    Hemisphere X=true
    MaskSpawnGroup=false
    Offset=<25 0 135>
    Override Seed=false
    Sphere Radius=&RandomRangeFloat(Minimum=10, Maximum=20)
    Surface Only=false
    UseNonUniformScale=false
    UseOffset=true
    UseSphereOrientation=false
   )
  ]
  particleUpdateScript [
   &ParticleState()
   &Drag(Drag=5, Use Linear Drag=true, Use Rotational Drag=false)
   &CurlNoiseForce(Noise Frequency=25, Noise Strength=150, Pan Noise=false)
   &SolveForcesAndVelocity(Clamp Velocity=false, Limit Acceleration=false)
   &ScaleColor(
    Scale Alpha=&FloatFromCurve(
     FloatCurve=DataInterface() {propertyValues {Curve {keys [
         {arriveTangent=4.2631, interpMode=RCIM_Cubic, leaveTangent=4.2631, tangentMode=RCTM_User, time=0, value=0}
         <at 0.3504 1 RCIM_Cubic>
         {arriveTangent=-0.1318, interpMode=RCIM_Cubic, leaveTangent=-0.1318, tangentMode=RCTM_User, time=1, value=0}
        ]}}}
     Scale Curve=&Power(a=&Multiply_Float(A=20, B=$Particles.Initial.MaterialRandom), b=0.5)
    )
    ScaleA=true
    ScaleRGB=true
   )
   &ScaleSpriteSize(Scale Factor=&Vector2DFromFloat(Value=&FloatFromCurve(
      FloatCurve=DataInterface() {propertyValues {Curve {keys [
          {arriveTangent=4.9073, interpMode=RCIM_Cubic, leaveTangent=4.9073, tangentMode=RCTM_User, time=0, value=0}
          <at 0.2737 1 RCIM_Cubic>
          {arriveTangent=-2.1114, interpMode=RCIM_Cubic, leaveTangent=-2.1114, tangentMode=RCTM_User, time=1, value=0}
         ]}}}
      CurveIndex=&MakeCustomFloatFromBool(
       Bool=&SetBoolByIntComparison(A=$Particles.SpawnGroup, B=3, Comparison Type=::NewEnumerator4)
       True Float=&Multiply_Float(A=$Particles.NormalizedAge, B=2)
       False Float=$Particles.NormalizedAge
      )
      Scale Curve=&MakeCustomFloatFromBool(
       Bool=&SetBoolByIntComparison(A=$Particles.SpawnGroup, B=3, Comparison Type=::NewEnumerator4)
       True Float=&RandomRangeFloat(Minimum=1, Maximum=2)
       False Float=1
      )
     )))
  ]
  rendererClasses [&NiagaraSpriteRendererProperties]
  renderers [
   {rendererClass=&NiagaraSpriteRendererProperties}
   {rendererClass=&NiagaraSpriteRendererProperties, rendererIndex=1}
  ]
 }}
summary {
 systemName=NS_Emote_Doodling_sparkles
 emitters [{emitterName=Sparks, rendererClasses [&NiagaraSpriteRendererProperties]}]
 userVariables [
  User.RibbonIDOffset {type=NiagaraInt32}
  User.SkeletalMesh {
   type=NiagaraDataInterfaceSkeletalMesh
   defaultValue {
    struct=NiagaraExt_VariableValue_DataInterface
    value {dataInterface=&self:NiagaraDataInterfaceSkeletalMesh_0}
   }
  }
 ]
}
