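# Lit material for the Doodling gadget emote prop, drawing a fadeable two-tone heart over the prop's diffuse texture.
# A scaled, rotated heart alpha masks the blend between heart colors and diffuse.
# HeartFade offsets a diamond gradient plus contrasted cloud noise within that mask, so raising it dissolves the heart unevenly.
# Separate mask and normal textures supply surface detail, and usage flags cover skeletal meshes, mesh deformers and Niagara sprites.
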Material /BRCosmetics/Animation/Props/Gadgets_Doodling/Materials/M_Gadget_Doodling
graph {
 properties {
  bUsedWithMeshDeformer=true
  bUsedWithNiagaraSprites=true
  bUsedWithSkeletalMesh=true
  pixelDepthOffsetMode=PDOM_Legacy
  usedShadingModels=Default Lit
 }
 nodes [
  Constant()
  n1 = TextureSample(UVs=MaterialFunctionCall(
    UVs=TextureCoordinate() {uTiling=1.7, vTiling=1.7}
    Rotation Center=Constant2Vector() {g=0.4, r=0.59}
    "Rotation Angle (0-1)"=Constant() {r=-0.75}
   ) {materialFunction=&CustomRotator}.Rotated Values) {texture=/BRCosmetics/Animation/Props/Gadgets_Doodling/FX/T_HeartIcon_Alpha}
  n2 = MaterialFunctionCall!(Falloff) {materialFunction=/Engine/Functions/Engine_MaterialFunctions01/Gradient/DiamondGradient}
  n3 = TextureCoordinate() {uTiling=0.1, vTiling=2.1576}
  n4 = Clamp(None=MaterialFunctionCall(In=TextureSample(UVs=@n3) {texture=&AtmosphericCloudNoise01}.RGB, Contrast=Constant() {r=0.5}) {materialFunction=&CheapContrast}.Result)
  n5 = ScalarParameter() {parameterName=HeartFade}
  n6 = Multiply(A=@n1.A, B=Add(A=Add(A=@n2.DiamondGradient, B=@n4), B=LinearInterpolate(Alpha=@n5) {constA=1, constB=-1}))
  LinearInterpolate(
   A=LinearInterpolate(A=Constant3Vector() {constant=<rgba 0.8353 0 0.1255 0>}, B=Constant3Vector() {constant=<rgba 0.6654 0.3152 0.3228 0>}, Alpha=@n1.RGB)
   B=TextureSample() {texture=/BRCosmetics/Animation/Props/Gadgets_Doodling/Textures/T_Emote_Prop_Doodling_D}.RGB
   Alpha=Clamp(None=OneMinus(None=@n6))
  )
  TextureSample() {samplerType=SAMPLERTYPE_Masks, texture=/BRCosmetics/Animation/Props/Gadgets_Doodling/Textures/T_Emote_Prop_Doodling_S}
  TextureSample() {samplerType=SAMPLERTYPE_Normal, texture=/BRCosmetics/Animation/Props/Gadgets_Doodling/Textures/T_Emote_Prop_Doodling_N}
 ]
}
